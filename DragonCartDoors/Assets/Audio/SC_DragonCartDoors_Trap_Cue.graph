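# Sound cue for the trap sound of the DragonCartDoors asset, playing one of three trap wave variations.
# A Random node at the root picks among the three wave players with equal weights.
# It routes to the Fort_Doors sound class with a boosted volume and uses a shared door attenuation with less reverb.
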SoundCue /DragonCartDoors/Assets/Audio/SC_DragonCartDoors_Trap_Cue
root=Random_0
nodes [
 {id=Random_0, children [WavePlayer_0,WavePlayer_1,WavePlayer_2], properties {weights [1,1,1]}}
 {id=WavePlayer_0, properties {soundWaveAssetPtr=&/DragonCartDoors_Trap_01}}
 {id=WavePlayer_1, properties {soundWaveAssetPtr=&/DragonCartDoors_Trap_02}}
 {id=WavePlayer_2, properties {soundWaveAssetPtr=&/DragonCartDoors_Trap_03}}
]
properties {
 attenuationSettings=/Game/Sounds/Attenuation/Fort_Doors_Less_Reverb_Att
 maxDistance=2400
 soundClassObject=&Fort_Doors
 volumeMultiplier=1.8
}
