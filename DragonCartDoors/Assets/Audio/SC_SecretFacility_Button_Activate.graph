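# Sound cue for the secret facility button activation in DragonCartDoors, playing a single button press wave.
# One wave player feeds the output directly, with no randomization or modulation.
# It routes through the gadgets sound class and borrows the Fortnite security camera alarm light attenuation asset for spatialization.
# A concurrency override caps how many instances play at once.
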SoundCue /DragonCartDoors/Assets/Audio/SC_SecretFacility_Button_Activate
root=WavePlayer_1
nodes [{id=WavePlayer_1, properties {soundWaveAssetPtr=&/DragonCartDoors_ButtonPress_01}}]
properties {
 attenuationSettings=/FNGameplayCues/Athena/Sounds/Gadgets/SecurityTurret/Security_Camera_Alarm_Light_Att
 bOverrideConcurrency=true
 maxDistance=6000
 soundClassObject=&Fort_Gadgets
 volumeMultiplier=1
 attenuationOverrides {distanceAlgorithm=Custom, falloffDistance=10112.3027, attenuationShapeExtents {x=600}}
 concurrencyOverrides {maxCount=3}
}
