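# Sound cue for the closing sound of a small Secret Facility door, reverse variant, in DragonCartDoors.
# A single wave player feeds the output directly, classed under Fort_Doors and set to keep playing when silent.
# Overridden attenuation adds HRTF spatialization, distance-based low-pass filtering and a send to the distant reverb submix.
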SoundCue /DragonCartDoors/Assets/Audio/SC_SecretFacility_Small_Door_Reverse_Close
root=WavePlayer_0
nodes [{id=WavePlayer_0, properties {soundWaveAssetPtr=&/SecretFacility_SmallDoor_Reverse_Close}}]
properties {
 attenuationSettings=/Game/Sounds/Attenuation/Fort_Doors_NoOcclusion_Att
 bOverrideAttenuation=true
 maxDistance=3300
 soundClassObject=&Fort_Doors
 virtualizationMode=PlayWhenSilent
 volumeMultiplier=1.3
 attenuationOverrides {
  bAttenuateWithLPF=true
  bEnableLogFrequencyScaling=true
  bEnableSubmixSends=true
  binauralRadius=800
  falloffDistance=2500
  lPFFrequencyAtMax=500
  lPFRadiusMax=2000
  lPFRadiusMin=400
  spatializationAlgorithm=SPATIALIZATION_HRTF
  stereoSpread=800
  attenuationShapeExtents {x=800}
  submixSendSettings [{
    bEnableHPFCutoff=false
    bEnableLPFCutoff=false
    disableManualSendClamp=false
    hPFCutoff=20
    lPFCutoff=20000
    maxSendDistance=7500
    maxSendLevel=0.25
    minSendDistance=0
    minSendLevel=0
    sendLevel=0.2
    sendLevelControlMethod=Linear
    soundSubmix=&Fortnite_Reverb_Distant
    customSendLevelCurve {externalCurve=/Game/Sounds/SubMixes/Reverb/Fortnite_Reverb_Distant_Send}
   }]
 }
}
