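# Default Lit parent material for flat, untextured surfaces, defined by parameters for material instances to override.
# It exposes a Color vector and scalar Metallic, Specular and Roughness values, with no texture sampling.
# An EnableEmissive static switch takes an EmissiveColor parameter, black by default, on both branches, so toggling it selects the same color either way.
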Material /DragonCartDoors/Assets/M_FlatColorParent
graph {
 properties {bAutomaticallySetUsageInEditor=false, usedShadingModels=Default Lit}
 nodes [
  ScalarParameter() {parameterName=Metallic}
  ScalarParameter() {defaultValue=0.5, parameterName=Specular}
  ScalarParameter() {defaultValue=0.5, parameterName=Roughness}
  StaticSwitchParameter(
   True=VectorParameter() {parameterName=EmissiveColor, defaultValue=<rgba 0 0 0 1>}.RGB
   False=VectorParameter() {parameterName=EmissiveColor, defaultValue=<rgba 0 0 0 1>}.RGB
  ) {parameterName=EnableEmissive}
  VectorParameter() {parameterName=Color, defaultValue=<rgba 1 1 1 1>}
 ]
}
